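# Unlit, Nanite-compatible MaskMaker material that acts as a Lumen-only light, as its name indicates.
# A RayTracingQualitySwitch returns black in normal rendering and the tintable Emissive parameter color on the ray-traced path, so it glows only there.
# That color is multiplied by a world-height mask that keeps it at and above the object's position and cuts it to zero within one unit below.
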Material /Hermes/Environments/TechArt/MaskMaker/M_MaskMaker_LumenOnlyLights2
graph {
 properties {bUsedWithNanite=true, pixelDepthOffsetMode=PDOM_Legacy, shadingModel=MSM_Unlit, usedShadingModels=Unlit}
 nodes [
  n1 = Subtract(A=ComponentMask(None=ObjectPositionWS()) {b=true}, B=WorldPosition().Z)
  Multiply(
   A=RayTracingQualitySwitch(Normal=Constant(), RayTraced=VectorParameter() {parameterName=Emissive, defaultValue=<rgba 1.4239 0.4688 1.5 1>}.RGB)
   B=OneMinus(None=Saturate(None=@n1))
  )
 ]
}
